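# Niagara system for the Ramen vehicle boost cosmetic, spawning trail effects at boost sockets on the car's skeletal mesh, shifted by a car offset.
# One ribbon emitter computes four sine-wobbled "noodle" positions per particle and has four ribbon renderers.
# Sprite emitters add droplets and two groups of spinning SubUV pieces, while a mesh-plus-sprite emitter grows and fades.
# Painted-variant colors tint all but the spinning pieces; variant ID also selects the mesh emitter's material values.
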NiagaraSystem /VehicleCosmetics/Boosts/Boost_Ramen/FX/NS_Boost_Ramen
dependencies {
 usedDataInterfaces [
  &NiagaraDataInterfaceArrayFloat
  &NiagaraDataInterfaceSpriteRendererInfo
  &NiagaraDataInterfaceCurve
  &NiagaraDataInterfaceVectorCurve
  &NiagaraDataInterfaceSkeletalMesh
 ]
 usedDynamicInputs []
 usedModules [
  &self:SystemSpawnScript.SystemScriptSource.SystemScriptGraph.NiagaraNodeAssignment_0.SetVariables_#1
  &SystemState
 ]
}
emitters {
 Drops {
  emitterUpdateScript [
   &EmitterState()
   &VC_FilteredSocketCount()
   &SpawnRate(SpawnRate=30, Use Spawn Probability=false)
  ]
  particleSpawnScript [
   &VC_PaintedVariantsOverride(
    BurntSienna=<rgba 0.151 0.029 0.0073 2>
    Cobalt=<rgba 0 0.0124 0.6719 2>
    Crimson=<rgba 0.224 0.0105 0.0163 2>
    DefaultColor=<rgba 0.4635 0.3351 0.1135 7>
    ForestGreen=<rgba 0 0.151 0.0076 2>
    Grey=<rgba 0.0938 0.0938 0.0938 2>
    Lime=<rgba 0.0964 0.1927 0 2>
    Orange=<rgba 0.1302 0.0454 0 2>
    Pink=<rgba 0.5052 0.0605 0.3435 2>
    Purple=<rgba 0.1176 0 0.2865 2>
    Saffron=<rgba 0.2188 0.1677 0 2>
    SkyBlue=<rgba 0 0.1391 1 2>
    TitaniumWhite=<rgba 0.4905 0.491 0.5208 2>
   )
   &InitializeParticle(
    Color Mode=::NewEnumerator0
    Lifetime Max=0.5
    Lifetime Min=0.4
    Lifetime Mode=::NewEnumerator1
    Sprite Rotation Mode=::NewEnumerator2
    Sprite Size=&RandomRangeVector2D(
     Minimum=&Multiply_Vector2D_ByFloat(Float=1, Vector2D=<5 5>)
     Maximum=&Multiply_Vector2D_ByFloat(Float=0.8, Vector2D=<7.8 7.8>)
    )
    Sprite Size Mode=::NewEnumerator4
    UsePositionOffset=false
    Write Material Random=true
   )
   &PartitionParticles()
   &SkeletalMeshLocation(
    Bone / Socket Index=$Particles.PartitionParticles.Partition
    Bone / Socket Sampling Mode=::NewEnumerator5
    MaskSpawnGroup=false
    Mesh Sampling Type=::NewEnumerator6
    Sampled Position Offset=&AddVectorToPosition(Position=<-75 0 0>, Vector=$User.DTCarOffset)
    Skeletal Mesh=$System.SkeletalMesh
   )
   &AddVelocity(Velocity=&RandomRangeVector(Minimum=<-80 35 -25>, Maximum=<-65 -35 -12>))
   &ShapeLocation(MaskSpawnGroup=false, Override Seed=false, Sphere Radius=30, UseOwnerScale=true)
  ]
  particleUpdateScript [
   &ParticleState()
   &ScaleSpriteSize(
    Uniform Curve Scale=2.5
    Uniform Curve Sprite Scale=DataInterface() {propertyValues {Curve {keys [
        {interpMode=RCIM_Cubic, tangentMode=RCTM_Break, time=0.2, value=1}
        {arriveTangent=-2.1805, interpMode=RCIM_Cubic, tangentMode=RCTM_Break, time=1, value=0.2}
       ]}}}
   )
   &ScaleVelocity(
    Coordinate Space:ENiagaraCoordinateSpace=::Simulation
    Velocity Scale=&VectorFromCurve(
     VectorCurve=DataInterface() {propertyValues {
       XCurve {
        defaultValue=1
        keys [
         {interpMode=RCIM_Cubic, tangentMode=RCTM_Break, time=0, value=1}
         {arriveTangentWeight=1.02, interpMode=RCIM_Cubic, tangentMode=RCTM_Break, time=1, value=1}
        ]
       }
       YCurve {
        defaultValue=1
        keys [
         {interpMode=RCIM_Cubic, tangentMode=RCTM_Break, time=0, value=1}
         {arriveTangentWeight=1.02, interpMode=RCIM_Cubic, tangentMode=RCTM_Break, time=1, value=1}
        ]
       }
       ZCurve {keys [
         {interpMode=RCIM_Cubic, leaveTangent=-3.0105, tangentMode=RCTM_Break, time=0.1, value=0}
         {arriveTangent=0.0661, arriveTangentWeight=1.02, interpMode=RCIM_Cubic, tangentMode=RCTM_Break, time=1, value=1}
        ]}
      }}
     Scale Curve=&RandomRangeVector(Minimum=<2 5 7>, Maximum=<4 3 4>)
    )
   )
   &SolveForcesAndVelocity(Clamp Velocity=false, Limit Acceleration=false)
   &DynamicMaterialParameters(
    Index 0 Param 1=&FloatFromCurve(FloatCurve=DataInterface() {propertyValues {Curve {keys [<at 0 1>,<at 1 1.5>]}}})
    Index 0 Param 2=&FloatFromCurve(FloatCurve=DataInterface() {propertyValues {Curve {keys [<at 0.45 -1>,<at 1 0>]}}})
    Index 0 Param 3=0.192
    Index 1 Param 1=1
    Index 1 Param 3=1
    Index 1 Param 4=2
    Module.Index0Float0Write=true
    Module.Index0Float1Write=true
    Module.Index0Float2Write=true
    Module.Index0Float3Write=true
    Module.Index1Float0Write=true
    Module.Index1Float1Write=true
    Module.Index1Float2Write=true
    Module.Index1Float3Write=true
    Write Parameter Index 1=::NewEnumerator1
   )
  ]
  rendererClasses [&NiagaraSpriteRendererProperties]
  renderers [{rendererClass=&NiagaraSpriteRendererProperties}]
 }
 NE_VC_Boost_Ribbon_Parent {
  emitterUpdateScript [
   &EmitterState(Life Cycle Mode=::NewEnumerator1, Loop Duration=1, UseLoopDelay=false)
   &VC_FilteredSocketCount()
   &VC_RateQualityHandler(High Spawn Rate=60, Low Spawn Rate=60, Medium Spawn Rate=60, Use Custom Rate?=true)
   &VC_SpawnRibbon_v1(Number of Ribbons=$Emitter.PortCount, SpawnRate=$Emitter.Spawn Rate)
  ]
  particleSpawnScript [
   &VC_PaintedVariantsOverride(
    BurntSienna=<rgba 0.08 0.0247 0.0099 0.8435>
    Cobalt=<rgba 0.0136 0.0112 0.6042 1>
    Crimson=<rgba 0.2552 0.0265 0.0367 1>
    DefaultColor=<rgba 0.8 0.3658 0.12 1>
    ForestGreen=<rgba 0.0089 0.2135 0.0192 1>
    Grey=<rgba 0.0677 0.0677 0.0677 1>
    Lime=<rgba 0.1224 0.2448 0 1>
    Orange=<rgba 0.2552 0.061 0.0106 1>
    Pink=<rgba 0.3281 0.0684 0.2337 1>
    Purple=<rgba 0.0953 0.024 0.2 1>
    Saffron=<rgba 0.3009 0.2609 0.0063 1>
    SkyBlue=<rgba 0 0.1492 0.4844 1>
   )
   &VC_InitializeRibbon_v1(
    Color=$Particles.Color
    Custom Offset=true
    Enable Skeletal Position=true
    Lifetime=0.4
    Local Vector Offset=&AddVectorToPosition(Position=<-25 0 0>, Vector=$User.DTCarOffset)
    Offset Amount=1
    Ribbon Width=4
   )
   &AddVelocity(Velocity=&MakeVector(X=-2000, Y=0, Z=0)) {module=AddVelocity001}
  ]
  particleUpdateScript [
   &ParticleState()
   &InheritSourceMovement(ApplyForce=false, ApplyPosition=true, ApplyPositionFalloff=false)
   &SolveForcesAndVelocity(Clamp Velocity=false, Limit Acceleration=false)
   &VC_UpdateRibbon_v1(Enable Skeletal Position=true)
   &ScaleColor(Scale RGB=&VectorFromFloat(Value=1.9), ScaleA=true, ScaleRGB=true)
   &ScaleRibbonWidth(Ribbon Width Scale=&FloatFromCurve(FloatCurve=DataInterface() {propertyValues {Curve {
        defaultValue=1
        keys [
         {tangentMode=RCTM_User, time=0, value=0}
         <at 0.05 1>
         {arriveTangent=-0.0374, interpMode=RCIM_Cubic, leaveTangent=-0.0374, tangentMode=RCTM_User, time=0.85, value=1}
         {arriveTangent=-12.8345, interpMode=RCIM_Cubic, leaveTangent=-12.8345, tangentMode=RCTM_User, time=0.89, value=0.6571}
         {arriveTangent=-68.1629, interpMode=RCIM_Cubic, leaveTangent=-68.1629, tangentMode=RCTM_User, time=0.9, value=0.4}
        ]
       }}}))
   /SetVariables(
    Particles.PosNoodle1:NiagaraPosition=&AddVectorToPosition(
     Position=$Particles.Position
     Vector=&TransformVector(
      Vector=&MakeVector(
       X=0
       Y=&Sine(
        Normalized Angle=&Add_Float(
         A=&Multiply_Float(A=$Particles.NormalizedAge, B=-1.3)
         B=&Multiply_Float(A=$Emitter.LoopedAge, B=1)
        )
        Period=&MaskFloatBySpawnGroup(Masked Float=0.6, Passthrough Float=-0.6)
        Scale=&FloatFromCurve(
         FloatCurve=DataInterface() {propertyValues {Curve {keys [{arriveTangent=0.1945, interpMode=RCIM_Cubic, leaveTangent=0.1945, tangentMode=RCTM_User, time=0, value=0}, <at 0.2717 1>, <at 1 5>]}}}
         Scale Curve=4
        )
        Bias=0
       )
       Z=&Sine(
        Normalized Angle=&Add_Float(
         A=&Multiply_Float(A=$Particles.NormalizedAge, B=-1.3)
         B=&Multiply_Float(A=$Emitter.LoopedAge, B=1)
        )
        Period=0.4
        Scale=&FloatFromCurve(FloatCurve=DataInterface() {propertyValues {Curve {keys [<at 0 0 RCIM_Cubic>,<at 0.2717 1>,<at 1 5>]}}}, Scale Curve=-6)
        Bias=0
       )
      )
      Destination Space=::Simulation
      Apply Scale=true
     )
    )
    Particles.PosNoodle2:NiagaraPosition=&AddVectorToPosition(
     Position=&AddVectorToPosition(
      Position=$Particles.Position
      Vector=&TransformVector(
       Vector=&MakeVector(
        X=-5
        Y=&Sine(
         Normalized Angle=&Add_Float(
          A=&Multiply_Float(A=$Particles.NormalizedAge, B=-1.3)
          B=&Multiply_Float(A=$Emitter.LoopedAge, B=1)
         )
         Period=&MaskFloatBySpawnGroup(Masked Float=0.5, Passthrough Float=-0.5)
         Scale=&FloatFromCurve(
          FloatCurve=DataInterface() {propertyValues {Curve {keys [{arriveTangent=0.1945, interpMode=RCIM_Cubic, leaveTangent=0.1945, tangentMode=RCTM_User, time=0, value=0}, <at 0.2717 1>, <at 1 5>]}}}
          Scale Curve=5
         )
         Bias=0
        )
        Z=&Sine(
         Normalized Angle=&Add_Float(
          A=&Multiply_Float(A=$Particles.NormalizedAge, B=-1.3)
          B=&Multiply_Float(A=$Emitter.LoopedAge, B=1)
         )
         Period=0.6
         Scale=&FloatFromCurve(FloatCurve=DataInterface() {propertyValues {Curve {keys [<at 0 0 RCIM_Cubic>,<at 0.2717 1>,<at 1 5>]}}}, Scale Curve=5)
         Bias=0
        )
       )
       Destination Space=::Simulation
       Apply Scale=true
      )
     )
     Vector=<0 -2.5 -2.5>
    )
    Particles.PosNoodle3:NiagaraPosition=&AddVectorToPosition(
     Position=&AddVectorToPosition(
      Position=$Particles.Position
      Vector=&TransformVector(
       Vector=&MakeVector(
        X=-10
        Y=&Sine(
         Normalized Angle=&Add_Float(
          A=&Multiply_Float(A=$Particles.NormalizedAge, B=-1.3)
          B=&Multiply_Float(A=$Emitter.LoopedAge, B=1)
         )
         Period=&MaskFloatBySpawnGroup(Masked Float=0.5, Passthrough Float=-0.5)
         Scale=&FloatFromCurve(
          FloatCurve=DataInterface() {propertyValues {Curve {keys [
              {arriveTangent=0.1945, interpMode=RCIM_Cubic, leaveTangent=0.1945, tangentMode=RCTM_User, time=0, value=0}
              {arriveTangent=0.5255, interpMode=RCIM_Cubic, leaveTangent=0.5255, tangentMode=RCTM_User, time=0.2717, value=1}
              <at 1 4>
             ]}}}
          Scale Curve=5
         )
         Bias=0
        )
        Z=&Sine(
         Normalized Angle=&Add_Float(
          A=&Multiply_Float(A=$Particles.NormalizedAge, B=-1.3)
          B=&Multiply_Float(A=$Emitter.LoopedAge, B=1)
         )
         Period=-0.5
         Scale=&FloatFromCurve(FloatCurve=DataInterface() {propertyValues {Curve {keys [<at 0 0 RCIM_Cubic>,<at 0.2717 1>,<at 1 4>]}}}, Scale Curve=5)
         Bias=0
        )
       )
       Destination Space=::Simulation
       Apply Scale=true
      )
     )
     Vector=<0 2.5 2.5>
    )
    Particles.PosNoodle4:NiagaraPosition=&AddVectorToPosition(
     Position=$Particles.Position
     Vector=&TransformVector(
      Vector=&Add_Vector(
       A=&MakeVector(
        X=-25
        Y=&Sine(
         Normalized Angle=&Add_Float(
          A=&Multiply_Float(A=$Particles.NormalizedAge, B=-1.3)
          B=&Multiply_Float(A=$Emitter.LoopedAge, B=1)
         )
         Period=&MaskFloatBySpawnGroup(Masked Float=0.6, Passthrough Float=-0.6)
         Scale=&FloatFromCurve(
          FloatCurve=DataInterface() {propertyValues {Curve {keys [{arriveTangent=0.1945, interpMode=RCIM_Cubic, leaveTangent=0.1945, tangentMode=RCTM_User, time=0, value=0}, <at 0.2717 1>, <at 1 5>]}}}
          Scale Curve=6
         )
         Bias=0
        )
        Z=&Sine(
         Normalized Angle=&Add_Float(
          A=&Multiply_Float(A=$Particles.NormalizedAge, B=-1.3)
          B=&Multiply_Float(A=$Emitter.LoopedAge, B=1)
         )
         Period=0.5
         Scale=&FloatFromCurve(FloatCurve=DataInterface() {propertyValues {Curve {keys [<at 0 0 RCIM_Cubic>,<at 0.2717 1>,<at 1 5>]}}}, Scale Curve=-5)
         Bias=0
        )
       )
       B=<0 2.5 -2.5>
      )
      Destination Space=::Simulation
      Apply Scale=true
     )
    )
   ) {isSetParameters=true}
  ]
  rendererClasses [&NiagaraRibbonRendererProperties]
  renderers [
   {rendererClass=&NiagaraRibbonRendererProperties}
   {rendererClass=&NiagaraRibbonRendererProperties, rendererIndex=1}
   {rendererClass=&NiagaraRibbonRendererProperties, rendererIndex=2}
   {rendererClass=&NiagaraRibbonRendererProperties, rendererIndex=3}
  ]
 }
 NE_VC_Boost_Sprite_Parent {
  emitterUpdateScript [
   &EmitterState()
   &VC_FilteredSocketCount()
   &SpawnRate(SpawnRate=15, Use Spawn Probability=false)
   &SpawnRate(SpawnGroup=1, SpawnRate=10, Use Spawn Probability=false) {module=SpawnRate001}
  ]
  particleSpawnScript [
   &VC_PaintedVariantsOverride() {enabled=false}
   &InitializeParticle(
    Color Mode=::NewEnumerator0
    Lifetime=0.8
    Sprite Rotation Mode=::NewEnumerator2
    Sprite Size Mode=::NewEnumerator3
    Uniform Sprite Size=&SelectFloatFromArray(Direct Array Index=&SpawnGroup(), Float Selection Array=DataInterface() {propertyValues {FloatData [37,27]}})
    UsePositionOffset=false
    Write Material Random=true
   )
   &PartitionParticles(Number Of Partitions=$Emitter.PortCount)
   /SetVariables(Particles.VisibilityTag=&SpawnGroup()) {isSetParameters=true}
   &SkeletalMeshLocation(
    Bone / Socket Index=$Particles.PartitionParticles.Partition
    Bone / Socket Sampling Mode=::NewEnumerator5
    MaskSpawnGroup=false
    Mesh Sampling Type=::NewEnumerator6
    Sampled Position Offset=&AddVectorToPosition(Position=<-70 0 0>, Vector=$User.DTCarOffset)
    Skeletal Mesh=$System.SkeletalMesh
   )
   &AddVelocity(Velocity=<-450 0 0>)
   &ShapeLocation(MaskSpawnGroup=false, Override Seed=false, Sphere Radius=20, UseOwnerScale=true)
   &SubUVAnimation(
    "Sprite Renderer "=DataInterface() {propertyValues {SpriteRenderer=&self:NiagaraEmitter_0.NiagaraSpriteRendererProperties_0}}
    SubUV Animation Mode:ENiagara_SubUVLookupModeV2=::NewEnumerator2
    UseEndFrame=false
    UseStartFrame=false
   )
  ]
  particleUpdateScript [
   &ParticleState()
   &ScaleSpriteSize(Uniform Curve Sprite Scale=DataInterface() {propertyValues {Curve {
       defaultValue=0
       keys [
        {interpMode=RCIM_Cubic, tangentMode=RCTM_User, time=0.15, value=1}
        {arriveTangent=-1.1718, interpMode=RCIM_Cubic, leaveTangent=-1.1718, tangentMode=RCTM_User, time=1, value=0.45}
       ]
      }}})
   &SpriteRotationRate(Rotation Rate=&RandomRangeFloat(
     Minimum=&SelectFloatFromArray(
      Array Sampling Mode=::NewEnumerator0
      Direct Array Index=?
      Float Selection Array=DataInterface() {propertyValues {FloatData [-500,-800]}}
      Random Seed=0
     )
     Maximum=&SelectFloatFromArray(
      Array Sampling Mode=::NewEnumerator0
      Direct Array Index=?
      Float Selection Array=DataInterface() {propertyValues {FloatData [500,850]}}
      Random Seed=0
     )
    ))
   &Drag(
    Drag=&FloatFromCurve(
     FloatCurve=DataInterface() {propertyValues {Curve {defaultValue=1, keys [{tangentMode=RCTM_User, time=0.3, value=0}, {tangentMode=RCTM_User, time=1, value=1}]}}}
     Scale Curve=2
    )
    Use Linear Drag=true
    Use Rotational Drag=true
   )
   &SolveForcesAndVelocity(Clamp Velocity=false, Limit Acceleration=false)
  ]
  rendererClasses [&NiagaraSpriteRendererProperties]
  renderers [
   {rendererClass=&NiagaraSpriteRendererProperties}
   {rendererClass=&NiagaraSpriteRendererProperties, rendererIndex=1}
  ]
 }
 NE_VC_Boost_Sprite_Parent002 {
  emitterSpawnScript [
   /SetVariables(Emitter.SUBSURFACEEXPPOWER:NiagaraFloat=&SelectFloatFromArray(
     Direct Array Index=$User.ColorVariantID
     Float Selection Array=DataInterface() {propertyValues {FloatData [-1,-0.5,-0.5,-0.5,1,0,-0.5,-0.5,-0.5,20,-0.5,50,50,-0.2,0]}}
    )) {isSetParameters=true}
   /SetVariables(Emitter.Roughness:NiagaraFloat=&SelectFloatFromArray(
     Direct Array Index=$User.ColorVariantID
     Float Selection Array=DataInterface() {propertyValues {FloatData [-10,2,2,2,3,0.5,2,2,2,10,1,2,10,0,-20]}}
    )) {isSetParameters=true}
   /SetVariables(Emitter.Metalness:NiagaraFloat=&SelectFloatFromArray(
     Direct Array Index=$User.ColorVariantID
     Float Selection Array=DataInterface() {propertyValues {FloatData [-2,0,0,0,0,1,-1,0,0,0,0,0,0,0,0]}}
    )) {isSetParameters=true}
  ]
  emitterUpdateScript [
   &EmitterState(Inactive Response=::NewEnumerator1, Life Cycle Mode=::NewEnumerator1, UseLoopDelay=false)
   &SpawnRate(SpawnGroup=1, SpawnRate=30, Use Spawn Probability=false) {module=SpawnRate001}
   &VC_FilteredSocketCount()
   &SpawnRate(
    SpawnRate=&MakeCustomFloatFromBool(Bool=$User.bIsFrontendPreview, True Float=28, False Float=44)
    Use Spawn Probability=false
   ) {enabled=false}
  ]
  particleSpawnScript [
   &VC_PaintedVariantsOverride(
    BurntSienna=<rgba 0.08 0.02 0.008 1>
    Cobalt=<rgba 0 0.0148 0.8 1>
    Crimson=<rgba 0.4 0.024 0.0428 1>
    DefaultColor=<rgba 1 0.4345 0.13 1>
    ForestGreen=<rgba 0.011 0.22 0.011 1>
    Grey=<rgba 0.085 0.085 0.085 1>
    Lime=<rgba 0.0955 0.2 0.01 1>
    Orange=<rgba 0.3 0.0501 0.006 1>
    Pink=<rgba 0.35 0.07 0.2482 1>
    Purple=<rgba 0.1514 0.019 0.38 1>
    Saffron=<rgba 0.4 0.3667 0 1>
    SkyBlue=<rgba 0 0.125 0.5 1>
    TitaniumWhite=<rgba 0.45 0.45 0.45 1>
   )
   &InitializeParticle(
    Color=$Particles.Color
    Lifetime Max=0.6
    Lifetime Min=0.6
    Lifetime Mode=::NewEnumerator1
    Mesh Scale=<1.5 1.5 2.1>
    Mesh Scale Mode=::NewEnumerator4
    Sprite Rotation Mode=::NewEnumerator2
    Sprite Size Mode=::NewEnumerator3
    UsePositionOffset=false
    Write Material Random=true
   )
   &PartitionParticles(Number Of Partitions=$Emitter.PortCount)
   &SkeletalMeshLocation(
    Bone / Socket Index=$Particles.PartitionParticles.Partition
    Bone / Socket Sampling Mode=::NewEnumerator5
    MaskSpawnGroup=false
    Mesh Sampling Type=::NewEnumerator6
    Orientation Sampling=::NewEnumerator1
    Sampled Position Offset=&AddVectorToPosition(Position=<-30 0 0>, Vector=$User.DTCarOffset)
    Skeletal Mesh=$System.SkeletalMesh
    Use Sampled Mesh Attribute as Ribbon ID=false
   )
   &ShapeLocation(
    MaskSpawnGroup=true
    Offset=<5 0 0>
    Offset Mode=::NewEnumerator0
    Override Seed=false
    Spawn Group Mask=2
    Sphere Radius=1
    UseOwnerScale=true
   )
   &AddVelocity(Velocity=<0 0 0>) {enabled=false}
   &InitialMeshOrientation(
    Mesh Orientation Mode=::NewEnumerator7
    Rotation=&RandomRangeVector(Minimum=<-1 0 0>, Maximum=<1 0 0>)
    Use Rotation Vector=true
   )
  ]
  particleUpdateScript [
   &ParticleState()
   &ScaleMeshSize(Scale Factor=&VectorFromCurve(
     VectorCurve=DataInterface() {propertyValues {
       XCurve {keys [
         {arriveTangent=0.2699, interpMode=RCIM_Cubic, leaveTangent=0.2699, leaveTangentWeight=0.3333, tangentMode=RCTM_User, time=0, value=0.5}
         {arriveTangent=2.9664, arriveTangentWeight=0.3333, interpMode=RCIM_Cubic, leaveTangent=2.9664, tangentMode=RCTM_User, time=1, value=2.5}
        ]}
       YCurve {keys [
         {arriveTangent=0.2699, interpMode=RCIM_Cubic, leaveTangent=0.2699, leaveTangentWeight=0.3333, tangentMode=RCTM_User, time=0, value=0.5}
         {arriveTangent=2.9664, arriveTangentWeight=0.3333, interpMode=RCIM_Cubic, leaveTangent=2.9664, tangentMode=RCTM_User, time=1, value=2.5}
        ]}
       ZCurve {keys [
         {arriveTangent=0.0111, interpMode=RCIM_Cubic, leaveTangent=0.0111, tangentMode=RCTM_User, time=0, value=1}
         {arriveTangent=2.4313, interpMode=RCIM_Cubic, leaveTangent=2.4313, tangentMode=RCTM_User, time=0.9992, value=4.5}
        ]}
      }}
     Scale Curve=<1.8 1.8 1>
    ))
   &SolveForcesAndVelocity(Clamp Velocity=false, Limit Acceleration=false)
   &ScaleColor(
    Scale Alpha=&FloatFromCurve(
     FloatCurve=DataInterface() {propertyValues {Curve {
        defaultValue=1
        keys [
         {arriveTangent=43.5199, interpMode=RCIM_Cubic, leaveTangent=43.5199, tangentMode=RCTM_User, time=0, value=1}
         {interpMode=RCIM_Cubic, tangentMode=RCTM_User, time=0.2, value=10}
         {arriveTangent=-11.25, interpMode=RCIM_Cubic, leaveTangent=-11.25, time=0.4, value=10}
         {arriveTangentWeight=0.2667, interpMode=RCIM_Cubic, leaveTangent=-1.7044, time=1, value=1}
        ]
       }}}
     Scale Curve=&SelectFloatFromArray(
      Array Sampling Mode=::NewEnumerator0
      Direct Array Index=?
      Float Selection Array=DataInterface() {propertyValues {FloatData [0.1,0.15]}}
      Random Seed=0
     )
    )
    Scale RGB=&VectorFromFloat(Value=&FloatFromCurve(FloatCurve=DataInterface() {propertyValues {Curve {
         defaultValue=1
         keys [{interpMode=RCIM_Cubic, tangentMode=RCTM_User, time=0, value=1}, {interpMode=RCIM_Cubic, tangentMode=RCTM_User, time=0.5, value=0.5}]
        }}}))
    ScaleA=true
    ScaleRGB=true
   ) {module=ScaleColor001}
  ]
  rendererClasses [&NiagaraMeshRendererProperties, &NiagaraSpriteRendererProperties]
  renderers [
   {rendererClass=&NiagaraMeshRendererProperties, rendererIndex=0}
   {rendererClass=&NiagaraSpriteRendererProperties, rendererIndex=1}
  ]
 }
}
summary {
 systemName=NS_Boost_Ramen
 emitters [
  {emitterName=NE_VC_Boost_Sprite_Parent, rendererClasses [&NiagaraSpriteRendererProperties]}
  {emitterName=Drops, rendererClasses [&NiagaraSpriteRendererProperties]}
  {emitterName=NE_VC_Boost_Ribbon_Parent, rendererClasses [&NiagaraRibbonRendererProperties]}
  {
   emitterName=NE_VC_Boost_Sprite_Parent002
   rendererClasses [&NiagaraMeshRendererProperties, &NiagaraSpriteRendererProperties]
  }
 ]
 userVariables [
  User.AddMultiply {type=NiagaraFloat, defaultValue=0, description=""}
  User.bIsFrontendPreview {type=NiagaraBool}
  User.bIsPainted {type=NiagaraBool}
  User.Brightness {type=NiagaraFloat, defaultValue=0, description=""}
  User.ColorVariantID {type=NiagaraInt32}
  User.DTCarOffset {type=Vector3f, description="", defaultValue {struct=Vector3f, value=<0 0 0>}}
  User.PaintedColor {type=LinearColor, description="", defaultValue {struct=LinearColor, value=<rgba 1 1 1 1>}}
  User.PaintedScalar {type=NiagaraFloat, defaultValue=1, description=""}
  User.PortCount {type=NiagaraInt32, defaultValue=2}
  User.SolidMask {type=NiagaraFloat, defaultValue=0, description=""}
 ]
}
